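# Editor utility list-row widget for an animation batch-duplicate tool, showing an anim montage name alongside a preview of its duplicate's name.
# On list item set, it stores the item, casts it to BP_AnimBatchDup_AnimMon_Item and displays the montage name.
# UpdateTargets stores source and target substrings and rebuilds the preview with a case-sensitive Replace.
# UpdateWidget refreshes both texts and shows the preview box only when Selected is true.
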EditorUtilityWidgetBlueprint /AnimInEdTools/AnimBatchDuplicate/Widgets/EUW_AnimBatchDup_AM_Item
graphs [EventGraph {nodes [
   n1 = UpdateTargets!() -> (OutputDelegate, then, Source AM:String, Target AM:String) {is=title K2Node_CustomEvent}
   > Set Source AM Replaceable(Source AM Replaceable:String=@n1.Source AM) -> (Output_Get:String)
   > Set Target AM Replaceable(Target AM Replaceable:String=@n1.Target AM) -> (Output_Get:String)
   > n2 = Cast To BP_AnimBatchDup_AnimMon_Item!(execute, Object=Get List Item Object() -> (List Item Object:Object Reference)) -> (then, CastFailed, AsBP Anim Batch Dup Anim Mon Item:BP Anim Batch Dup Anim Mon Item Object Reference) {is=title K2Node_DynamicCast}
   n3 = Get Source AM Replaceable() -> (Source AM Replaceable:String)
   n4 = Get Target AM Replaceable() -> (Target AM Replaceable:String)
   n5 = Get Anim Montage Name!(self:BP Anim Batch Dup Anim Mon Item Object Reference=@n2) -> (Anim Montage Name:String)
   > Set Text(
    Text="To Text (String)"(InString=Replace(SourceString=@n5, From=@n3, To=@n4, SearchCase=CaseSensitive))
    self=Get TargetAnimMontageText() -> (TargetAnimMontageText:Text Object Reference)
   )
   n6 = UpdateWidget!() -> (OutputDelegate, then, Selected:bool=~) {is=title K2Node_CustomEvent}
   > n7 = Sequence()
   > n8 = Cast To BP_AnimBatchDup_AnimMon_Item(Object=Get List Item Object())
   n9 = Get Anim Montage Name(self=@n8)
   > Set Text(Text="To Text (String)"(InString=@n9), self=Get AnimMontageText() -> (AnimMontageText:Text Object Reference))
   n10 = Get Source AM Replaceable()
   n11 = Get Target AM Replaceable()
   > Set Text(Text="To Text (String)"(InString=Replace(SourceString=@n9, From=@n10, To=@n11, SearchCase=CaseSensitive)), self=Get TargetAnimMontageText())
   SetVisibility(
    execute=@n7.then_1
    self=Get TargetBox() -> (TargetBox:Horizontal Box Object Reference)
    InVisibility=Select(Option 0=Collapsed, Option 1=Visible, Index=@n6.Selected)
   )
   Event Construct()
   > SetVisibility(self=Get TargetBox(), InVisibility=Collapsed)
   n12 = Event OnListItemObjectSet()
   > Set List Item Object(List Item Object:Object Reference=@n12.ListItemObject) -> (Output_Get:Object Reference)
   > UpdateWidget!(execute, self, Selected:bool=false) -> (then) {is=title K2Node_CallFunction}
  ]}]
variables [
 List Item Object {type=Object, default=null}
 Source AM Replaceable {type=string, default=""}
 Target AM Replaceable {type=string, default=""}
]
widgets {
 info {rootWidgetClass=&StackBox, widgetCount=8}
 widgets [
  StackBox_0 {widgetClassPath=&StackBox, widget=StackBox_0}
  VerticalBox_0 {
   widgetClassPath=&VerticalBox
   parent=StackBox_0
   widget=VerticalBox_0
   slot {class=StackBoxSlot, properties {size {sizeRule=Fill, value=1}}}
  }
  HorizontalBox_0 {
   widgetClassPath=&HorizontalBox
   parent=VerticalBox_0
   widget=HorizontalBox_0
   slot {class=VerticalBoxSlot, properties {verticalAlignment=VAlign_Center}}
  }
  AnimMontageText {
   widgetClassPath=&TextBlock
   bIsVariable=true
   parent=HorizontalBox_0
   widget=AnimMontageText
   properties {
    text=Text Block
    colorAndOpacity {colorUseRule=UseColor_Specified, specifiedColor=<rgba 0.1274 0.1274 1 1>}
    font {
     bForceMonospaced=false
     bMaterialIsStencil=false
     fontMaterial=null
     fontObject=&Roboto
     letterSpacing=0
     monospacedWidth=1
     size=10.5
     skewAmount=0
     typefaceFontName=Bold
     outlineSettings {
      bApplyOutlineToDropShadows=false
      bMiteredCorners=false
      bSeparateFillAlpha=false
      outlineMaterial=null
      outlineSize=0
      outlineColor=<rgba 0 0 0 1>
     }
    }
   }
   slot {class=HorizontalBoxSlot, properties {size {sizeRule=Fill, value=1}}}
  }
  TargetBox {
   widgetClassPath=&HorizontalBox
   bIsVariable=true
   parent=HorizontalBox_0
   widget=TargetBox
   slot {class=HorizontalBoxSlot, properties {size {sizeRule=Fill, value=1}}}
  }
  HorizontalBox_1 {
   widgetClassPath=&HorizontalBox
   parent=TargetBox
   widget=HorizontalBox_1
   slot {class=HorizontalBoxSlot, properties {padding=<ltrb 20 0 0 0>}}
  }
  Image_83 {
   widgetClassPath=&Image
   bIsVariable=true
   parent=HorizontalBox_1
   widget=Image_83
   properties {
    brush {
     bIsDynamicallyLoaded=false
     drawAs=Image
     imageType=FullColor
     mirroring=NoMirror
     resourceName=None
     resourceObject=/AnimInEdTools/AnimBatchDuplicate/Icons/Icon_add
     tiling=NoTile
     imageSize=<14 14>
     margin=<ltrb 0 0 0 0>
     outlineSettings {
      bUseBrushTransparency=false
      roundingType=HalfHeightRadius
      width=0
      color {colorUseRule=UseColor_Specified, specifiedColor=<rgba 0 0 0 0>}
      cornerRadii=<0 0 0 0>
     }
     tintColor {colorUseRule=UseColor_Specified, specifiedColor=<rgba 1 1 1 1>}
     uVRegion {bIsValid=false, max=<0 0>, min=<0 0>}
    }
    colorAndOpacity=<rgba 0.016 0.2542 0.4852 1>
   }
   slot {class=HorizontalBoxSlot, properties {horizontalAlignment=HAlign_Center, verticalAlignment=VAlign_Center}}
  }
  TargetAnimMontageText {
   widgetClassPath=&TextBlock
   bIsVariable=true
   parent=HorizontalBox_1
   widget=TargetAnimMontageText
   properties {
    text=Text Block
    colorAndOpacity {colorUseRule=UseColor_Specified, specifiedColor=<rgba 0.1274 0.1274 1 1>}
    font {
     bForceMonospaced=false
     bMaterialIsStencil=false
     fontMaterial=null
     fontObject=&Roboto
     letterSpacing=0
     monospacedWidth=1
     size=10.5
     skewAmount=0
     typefaceFontName=Italic
     outlineSettings {
      bApplyOutlineToDropShadows=false
      bMiteredCorners=false
      bSeparateFillAlpha=false
      outlineMaterial=null
      outlineSize=0
      outlineColor=<rgba 0 0 0 1>
     }
    }
   }
   slot {class=HorizontalBoxSlot, properties {padding=<ltrb 5 0 0 0>, size {sizeRule=Fill, value=1}}}
  }
 ]
}
