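# Editor utility list-row widget for the animation batch duplicate tool, showing an Anim Sequence name beside a preview of its duplicate's name.
# On item assignment it casts to BP_AnimBatchDup_AnimSeq_Item and fills both texts; Update Widget's Selected flag toggles the target preview.
# UpdateTargets stores source/target substrings and refreshes the case-sensitive replacement, ignoring its montage arguments.
# A helper, uncalled here, loads Asset Registry referencers to collect a sequence's Anim Montages.
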EditorUtilityWidgetBlueprint /AnimInEdTools/AnimBatchDuplicate/Widgets/EUW_AnimBatchDup_AS_Item
graphs [
 EventGraph {nodes [
   n1 = Update Widget!() -> (OutputDelegate, then, Selected:bool=~) {is=title K2Node_CustomEvent}
   > n2 = Sequence()
   SetVisibility(
    execute=@n2.then_1
    self=Get TargetASBox() -> (TargetASBox:Vertical Box Object Reference)
    InVisibility=Select(Option 0=Collapsed, Option 1=Visible, Index=@n1.Selected)
   )
   n3 = Cast To BP_AnimBatchDup_AnimSeq_Item!(execute=@n2.then_0, Object=Get List Item Object() -> (List Item Object:Object Reference)) -> (then, CastFailed, AsBP Anim Batch Dup Anim Seq Item:BP Anim Batch Dup Anim Seq Item Object Reference) {is=title K2Node_DynamicCast}
   n4 = Get Anim Sequence Name!(self:BP Anim Batch Dup Anim Seq Item Object Reference=@n3) -> (Anim Sequence Name:String)
   > Set Text(Text="To Text (String)"(InString=@n4), self=Get AnimationSequenceText() -> (AnimationSequenceText:Text Object Reference))
   n5 = Get Source AS Replaceable() -> (Source AS Replaceable:String)
   n6 = Get Target AS Replaceable() -> (Target AS Replaceable:String)
   > Set Text(
    Text="To Text (String)"(InString=Replace(SourceString=@n4, From=@n5, To=@n6, SearchCase=CaseSensitive))
    self=Get TargetAnimationSequenceText() -> (TargetAnimationSequenceText:Text Object Reference)
   )
   n7 = UpdateTargets!() -> (OutputDelegate, then, Source AS:String, Target AS:String, Source AM:String, Target AM:String) {is=title K2Node_CustomEvent}
   > Sequence()
   > Set Source AS Replaceable(Source AS Replaceable:String=@n7.Source AS) -> (Output_Get:String)
   > Set Target AS Replaceable(Target AS Replaceable:String=@n7.Target AS) -> (Output_Get:String)
   > n8 = Cast To BP_AnimBatchDup_AnimSeq_Item(Object=Get List Item Object())
   n9 = Get Source AS Replaceable()
   n10 = Get Target AS Replaceable()
   n11 = Get Anim Sequence Name(self=@n8)
   > Set Text(
    Text="To Text (String)"(InString=Replace(SourceString=@n11, From=@n9, To=@n10, SearchCase=CaseSensitive))
    self=Get TargetAnimationSequenceText()
   )
   Event Construct()
   > SetVisibility(self=Get TargetASBox(), InVisibility=Collapsed)
   n12 = Event OnListItemObjectSet()
   > Set List Item Object(List Item Object:Object Reference=@n12.ListItemObject) -> (Output_Get:Object Reference)
   > Update Widget!(execute, self, Selected:bool=false) -> (then) {is=title K2Node_CallFunction}
  ]}
 Get AnimMontages from AnimSequence {nodes [
   Get AnimMontages from AnimSequence!() -> (then, Anim Sequence:Anim Sequence Object Reference) {is=title K2Node_FunctionEntry}
   > n1 = Sequence()
   n2 = GetPackageForObject!(Object:Object Reference=Get Anim Sequence() -> (Anim Sequence:Anim Sequence Object Reference)) -> (ReturnValue:Package Object Reference)
   > n3 = Get Referencers!(
    execute
    self:Asset Registry Interface=GetAssetRegistry()
    PackageName:Name=String To Name(InString=GetObjectName(Object=@n2))
    ReferenceOptions:"Asset Registry Dependency Options Structure (by ref)"=Make Asset Registry Dependency Options(bIncludeSoftPackageReferences=True)
   ) -> (then, OutReferencers:Array of Names, ReturnValue:bool) {is=title K2Node_CallFunction}
   > n4 = For Each Loop(Array:Array of Names=@n3) -> (Array Element:Name)
   > n5 = LoadAsset(AssetPath="To String (Name)"(InName=@n4))
   > n6 = Cast To AnimMontage!(execute, Object=@n5) -> (then, CastFailed, AsAnim Montage:Anim Montage Object Reference) {is=title K2Node_DynamicCast}
   > Add(
    TargetArray:Array of Anim Montage Object References=Get Local Anim Montages() -> (Local Anim Montages:Array of Anim Montage Object References)
    NewItem:"Anim Montage Object Reference (by ref)"=@n6
   ) {is=title K2Node_CallArrayFunction}
   Return Node!(execute=@n1.then_1, Anim Montages:Array of Anim Montage Object References=Get Local Anim Montages())
  ]}
]
variables [
 List Item Object {type=Object, default=null}
 Source AS Replaceable {type=string, default=""}
 Target AS Replaceable {type=string, default=""}
]
widgets {
 info {rootWidgetClass=&StackBox, widgetCount=8}
 widgets [
  StackBox_0 {widgetClassPath=&StackBox, widget=StackBox_0}
  VerticalBox_0 {
   widgetClassPath=&VerticalBox
   parent=StackBox_0
   widget=VerticalBox_0
   slot {class=StackBoxSlot, properties {size {sizeRule=Fill, value=1}}}
  }
  HorizontalBox_0 {widgetClassPath=&HorizontalBox, parent=VerticalBox_0, widget=HorizontalBox_0, slot {class=VerticalBoxSlot}}
  AnimationSequenceText {
   widgetClassPath=&TextBlock
   bIsVariable=true
   parent=HorizontalBox_0
   widget=AnimationSequenceText
   properties {
    text=Text Block
    colorAndOpacity {colorUseRule=UseColor_Specified, specifiedColor=<rgba 0.0802 0.1981 0.0648 1>}
    font {
     bForceMonospaced=false
     bMaterialIsStencil=false
     fontMaterial=null
     fontObject=&Roboto
     letterSpacing=0
     monospacedWidth=1
     size=10.5
     skewAmount=0
     typefaceFontName=Bold
     outlineSettings {
      bApplyOutlineToDropShadows=false
      bMiteredCorners=false
      bSeparateFillAlpha=false
      outlineMaterial=null
      outlineSize=0
      outlineColor=<rgba 0 0 0 1>
     }
    }
   }
   slot {class=HorizontalBoxSlot, properties {verticalAlignment=VAlign_Center, size {sizeRule=Fill, value=1}}}
  }
  TargetASBox {
   widgetClassPath=&VerticalBox
   bIsVariable=true
   parent=HorizontalBox_0
   widget=TargetASBox
   properties {visibility=Visible}
   slot {class=HorizontalBoxSlot, properties {size {sizeRule=Fill, value=1}}}
  }
  HorizontalBox_1 {
   widgetClassPath=&HorizontalBox
   parent=TargetASBox
   widget=HorizontalBox_1
   slot {class=VerticalBoxSlot, properties {padding=<ltrb 20 0 0 0>}}
  }
  Image_83 {
   widgetClassPath=&Image
   bIsVariable=true
   parent=HorizontalBox_1
   widget=Image_83
   properties {
    brush {
     bIsDynamicallyLoaded=false
     drawAs=Image
     imageType=FullColor
     mirroring=NoMirror
     resourceName=None
     resourceObject=/AnimInEdTools/AnimBatchDuplicate/Icons/Icon_add
     tiling=NoTile
     imageSize=<14 14>
     margin=<ltrb 0 0 0 0>
     outlineSettings {
      bUseBrushTransparency=false
      roundingType=HalfHeightRadius
      width=0
      color {colorUseRule=UseColor_Specified, specifiedColor=<rgba 0 0 0 0>}
      cornerRadii=<0 0 0 0>
     }
     tintColor {colorUseRule=UseColor_Specified, specifiedColor=<rgba 1 1 1 1>}
     uVRegion {bIsValid=false, max=<0 0>, min=<0 0>}
    }
    colorAndOpacity=<rgba 0.016 0.2542 0.4852 1>
   }
   slot {class=HorizontalBoxSlot, properties {horizontalAlignment=HAlign_Center, verticalAlignment=VAlign_Center}}
  }
  TargetAnimationSequenceText {
   widgetClassPath=&TextBlock
   bIsVariable=true
   parent=HorizontalBox_1
   widget=TargetAnimationSequenceText
   properties {
    text=Text Block
    colorAndOpacity {colorUseRule=UseColor_Specified, specifiedColor=<rgba 0.0802 0.1981 0.0648 1>}
    font {
     bForceMonospaced=false
     bMaterialIsStencil=false
     fontMaterial=null
     fontObject=&Roboto
     letterSpacing=0
     monospacedWidth=1
     size=10.5
     skewAmount=0
     typefaceFontName=Italic
     outlineSettings {
      bApplyOutlineToDropShadows=false
      bMiteredCorners=false
      bSeparateFillAlpha=false
      outlineMaterial=null
      outlineSize=0
      outlineColor=<rgba 0 0 0 1>
     }
    }
   }
   slot {class=HorizontalBoxSlot, properties {padding=<ltrb 5 0 0 0>}}
  }
 ]
}
